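# Sound cue for the Bubblegum cosmetic pickaxe's impact, the 02 Player variant.
# A single wave player feeds the root directly with one impact wave, with no randomization or modulation.
# Shared pickaxe impact attenuation, cosmetic audio property sheet, impacts sound class and player melee weapon concurrency set govern playback.
# The cue primes on load.
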SoundCue /BRCosmetics/Sounds/PickAxes/Bubblegum/Pickaxe_Bubblegum_Impact_Player_02_Cue
root=WavePlayer_0
nodes [{id=WavePlayer_0, properties {soundWaveAssetPtr=&/BubbleGum_Pickaxe_Impact_02}}]
properties {
 attenuationSettings=/Game/Sounds/Attenuation/Fort_Pickaxe_Impact_Att
 audioPropertiesSheet=/BRCosmetics/Sounds/Settings/PropertySheets/Pickaxe/APS_Cosmetic_Pickaxe_Impact
 bPrimeOnLoad=true
 maxDistance=4300
 soundClassObject=&Fort_Impacts
 attenuationOverrides {
  bAttenuateWithLPF=true
  distanceAlgorithm=Logarithmic
  falloffDistance=800
  lPFRadiusMax=800
  lPFRadiusMin=400
  attenuationShapeExtents {x=300}
 }
 concurrencyOverrides {bLimitToOwner=true, maxCount=2, resolutionRule=StopOldest}
 concurrencySet [&Player_Melee_Weapon_Conc]
}
